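# Character material for the TractorFair male medium body, with emissive that fades out over time for a reactivity effect.
# The shared 2023 PreFX character function builds the base attributes, and only Emissive Color is overridden before they enter the matching PostFX function.
# Emissive is scaled by one minus saturated (Time minus Reactivity_Time) over Reactivity_PhaseTime, falling linearly from full to zero after that timestamp.
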Material /BRCosmetics/Characters/Player/Male/Medium/Bodies/M_MED_TractorFair/Materials/M_TractoFair_Reactivity
graph {
 properties {
  bUseMaterialAttributes=true
  bUsedWithMeshDeformer=true
  bUsedWithMorphTargets=true
  bUsedWithSkeletalMesh=true
  pixelDepthOffsetMode=PDOM_Legacy
  shadingModel=MSM_FromMaterialExpression
  usedShadingModels=Default Lit | Subsurface | Clear Coat | Subsurface Profile | Cloth
 }
 nodes [
  n1 = MaterialFunctionCall() {materialFunction=&MF_CharacterParent_2023_PreFX}
  n2 = GetMaterialAttributes(None=@n1.Result) {attributeGetTypes [#1]}
  n3 = Subtract(A=Time(), B=ScalarParameter() {group=Reactivity, parameterName=Reactivity_Time})
  n4 = ScalarParameter() {defaultValue=4, group=Reactivity, parameterName=Reactivity_PhaseTime}
  n5 = OneMinus(None=Saturate(None=Divide(A=@n3, B=@n4)))
  MaterialFunctionCall(
   Material=SetMaterialAttributes(MaterialAttributes=@n2.MaterialAttributes, Emissive Color=Multiply(A=@n2.Emissive Color, B=@n5)) {attributeSetTypes [#1]}
   M=@n1.M
   Pre Skinned Local Position=@n1.M
   Distance Blend=@n1.M
   Fresnel=@n1.M
  ) {materialFunction=&MF_CharacterParent_2023_PostFX}
 ]
}
