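# Character body material for the TractorFair player cosmetic that adds a reactive emissive glow to the shared 2023 character material.
# A tinted ReactivityEM mask, scaled by brightness, fades linearly to zero over a phase time measured from the Reactivity_Time parameter.
# A showcase parameter blends that fade toward a front-end intensity, and IsReactive gates the glow.
# The glow adds to emissive between the parent PreFX and PostFX material functions.
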Material /BRCosmetics/Characters/Player/Male/Medium/Bodies/M_MED_TractorFair/Materials/M_TractoFair_Reactivity_Body
graph {
 properties {
  bUseMaterialAttributes=true
  bUsedWithMeshDeformer=true
  bUsedWithMorphTargets=true
  bUsedWithNiagaraMeshParticles=true
  bUsedWithSkeletalMesh=true
  pixelDepthOffsetMode=PDOM_Legacy
  shadingModel=MSM_FromMaterialExpression
  usedShadingModels=Default Lit | Subsurface | Clear Coat | Subsurface Profile | Cloth
 }
 nodes [
  n1 = MaterialFunctionCall() {materialFunction=&MF_CharacterParent_2023_PreFX}
  n2 = GetMaterialAttributes(None=@n1.Result) {attributeGetTypes [#1]}
  n3 = TextureSampleParameter2D() {
   group=Reactivity
   parameterName=ReactivityEM
   texture=/BRCosmetics/Characters/Player/Male/Medium/Bodies/M_MED_TractorFair/Textures/T_TractorFair_Body_ReactiveEM
  }
  n4 = Multiply(A=@n3.RGB, B=VectorParameter() {group=Reactivity, parameterName=EmissiveTint_React, defaultValue=<rgba 0.7726 0.8632 0.2248 1>}.RGB)
  n5 = ScalarParameter() {defaultValue=2, parameterName=EmissiveBrightness}
  n6 = Subtract(A=Time(), B=ScalarParameter() {group=Reactivity, parameterName=Reactivity_Time})
  n7 = ScalarParameter() {defaultValue=4, group=Reactivity, parameterName=Reactivity_PhaseTime}
  n8 = OneMinus(None=Saturate(None=Divide(A=@n6, B=@n7)))
  n9 = ScalarParameter() {group=Reactivity, parameterName=FrontEnd_Intensity}
  n10 = ScalarParameter() {group=Reactivity, parameterName=IsFrontEnd_ShowCase}
  n11 = LinearInterpolate(
   B=Multiply(A=Multiply(A=@n4, B=@n5), B=LinearInterpolate(A=@n8, B=@n9, Alpha=@n10))
   Alpha=ScalarParameter() {defaultValue=1, group=Reactivity, parameterName=IsReactive}
  )
  MaterialFunctionCall(
   Material=SetMaterialAttributes(MaterialAttributes=@n2.MaterialAttributes, Emissive Color=Add(A=@n2.Emissive Color, B=@n11)) {attributeSetTypes [#1]}
   M=@n1.M
   Pre Skinned Local Position=@n1.M
   Distance Blend=@n1.M
   Fresnel=@n1.M
  ) {materialFunction=&MF_CharacterParent_2023_PostFX}
 ]
}
