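# Physics asset for the M_MED_TractorFair male player body cosmetic, setting up physics for five bags around the hips.
# Each bag has a tiny base capsule constrained to the pelvis and a larger dyn capsule constrained to that base with limited swing and very little twist.
# Pelvis, spine_01 and thigh capsules form the body volume, linked by angularly free constraints.
# Base-to-pelvis constraints are also angularly free, except the limited b_r one.
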PhysicsAsset /BRCosmetics/Characters/Player/Male/Medium/Bodies/M_MED_TractorFair/Meshes/M_MED_TractorFair_Physics
bodies {
 base_TF_bag_a_l [{shapeType=Capsule, length=1, radius=1, shapeName=base_TF_bag_a_l_capsule, rotation=<rot -90 -90 0>}]
 base_TF_bag_a_r [{shapeType=Capsule, length=1, radius=1, shapeName=base_TF_bag_a_r_capsule, rotation=<rot -90 -90 0>}]
 base_TF_bag_b_l [{shapeType=Capsule, length=1, radius=1, shapeName=base_TF_bag_b_l_capsule, rotation=<rot -90 -90 0>}]
 base_TF_bag_b_r [{shapeType=Capsule, length=1, radius=1, shapeName=base_TF_bag_b_r_capsule, rotation=<rot -90 -90 0>}]
 base_TF_bag_c_l [{shapeType=Capsule, length=1, radius=1, shapeName=base_TF_bag_c_l_capsule, rotation=<rot -90 -90 0>}]
 dyn_TF_bag_a_l [{shapeType=Capsule, length=6, radius=2, shapeName=dyn_TF_bag_a_l_capsule, center=<3.1688 0 1.8>, rotation=<rot -90 180 180>}]
 dyn_TF_bag_a_r [{shapeType=Capsule, length=6, radius=2, shapeName=dyn_TF_bag_a_r_capsule, center=<3 2 3.5>, rotation=<rot -90 0 0>}]
 dyn_TF_bag_b_l [{shapeType=Capsule, length=7, radius=3, shapeName=dyn_TF_bag_b_l_capsule, center=<7.8941 0 1>, rotation=<rot -90 175.2364 -85.2362>}]
 dyn_TF_bag_b_r [{shapeType=Capsule, length=6, radius=3, shapeName=dyn_TF_bag_b_r_capsule, center=<3 0 2.6087>, rotation=<rot -90 0 0>}]
 dyn_TF_bag_c_l [{shapeType=Capsule, length=6, radius=1.5, shapeName=dyn_TF_bag_c_l_capsule, center=<3 0 1>, rotation=<rot -90 0 0>}]
 pelvis [{shapeType=Capsule, length=11.7582, radius=11, shapeName=pelvis_capsule}]
 spine_01 [{shapeType=Capsule, length=11.0408, radius=11.0408, shapeName=spine_01_capsule, center=<0 2.4568 0>}]
 thigh_l [{shapeType=Capsule, length=10, radius=9.5, shapeName=thigh_l_capsule, center=<-8 0.621 0>, rotation=<rot 90 0 0>}]
 thigh_r [{shapeType=Capsule, length=10, radius=11, shapeName=thigh_r_capsule, center=<8 -0.6215 0>, rotation=<rot 90 0 0>}]
}
constraints [
 {
  bone1Name=spine_01
  bone2Name=pelvis
  swing1LimitDegrees=45
  swing1Motion=Free
  swing2LimitDegrees=45
  swing2Motion=Free
  twistLimitDegrees=45
  twistMotion=Free
 }
 {
  bone1Name=thigh_l
  bone2Name=pelvis
  swing1LimitDegrees=45
  swing1Motion=Free
  swing2LimitDegrees=45
  swing2Motion=Free
  twistLimitDegrees=45
  twistMotion=Free
 }
 {
  bone1Name=thigh_r
  bone2Name=pelvis
  swing1LimitDegrees=45
  swing1Motion=Free
  swing2LimitDegrees=45
  swing2Motion=Free
  twistLimitDegrees=45
  twistMotion=Free
 }
 {bone1Name=dyn_TF_bag_c_l, bone2Name=base_TF_bag_c_l, swing1LimitDegrees=15, swing2LimitDegrees=25, twistLimitDegrees=1}
 {bone1Name=dyn_TF_bag_b_r, bone2Name=base_TF_bag_b_r, swing1LimitDegrees=15, swing2LimitDegrees=25, twistLimitDegrees=1}
 {bone1Name=dyn_TF_bag_b_l, bone2Name=base_TF_bag_b_l, swing1LimitDegrees=15, swing2LimitDegrees=45, twistLimitDegrees=1}
 {bone1Name=dyn_TF_bag_a_r, bone2Name=base_TF_bag_a_r, swing1LimitDegrees=15, swing2LimitDegrees=65, twistLimitDegrees=1}
 {bone1Name=dyn_TF_bag_a_l, bone2Name=base_TF_bag_a_l, swing1LimitDegrees=1, swing2LimitDegrees=65, twistLimitDegrees=1}
 {
  bone1Name=base_TF_bag_c_l
  bone2Name=pelvis
  swing1LimitDegrees=45
  swing1Motion=Free
  swing2LimitDegrees=45
  swing2Motion=Free
  twistLimitDegrees=45
  twistMotion=Free
 }
 {bone1Name=base_TF_bag_b_r, bone2Name=pelvis, swing1LimitDegrees=25, swing2LimitDegrees=25, twistLimitDegrees=1}
 {
  bone1Name=base_TF_bag_b_l
  bone2Name=pelvis
  swing1LimitDegrees=45
  swing1Motion=Free
  swing2LimitDegrees=45
  swing2Motion=Free
  twistLimitDegrees=45
  twistMotion=Free
 }
 {
  bone1Name=base_TF_bag_a_r
  bone2Name=pelvis
  swing1LimitDegrees=45
  swing1Motion=Free
  swing2LimitDegrees=45
  swing2Motion=Free
  twistLimitDegrees=45
  twistMotion=Free
 }
 {
  bone1Name=base_TF_bag_a_l
  bone2Name=pelvis
  swing1LimitDegrees=45
  swing1Motion=Free
  swing2LimitDegrees=45
  swing2Motion=Free
  twistLimitDegrees=45
  twistMotion=Free
 }
]
