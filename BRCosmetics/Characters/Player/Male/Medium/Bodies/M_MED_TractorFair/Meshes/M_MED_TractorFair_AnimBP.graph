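# Anim Blueprint for the TractorFair male body cosmetic, adding deform and secondary motion to shoulder chains, belt and armbands.
# The copied parent pose feeds an emote-aware Control Rig, then weighted bone constraints.
# In front end one RigidBody simulates; otherwise another RigidBody feeds Anim Dynamics, with low-armband motion eased out while skydiving.
# Events update gravity and swap two material hide elements on weapon change or customization load, by tagged pickaxe or skydiving.
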AnimBlueprint /BRCosmetics/Characters/Player/Male/Medium/Bodies/M_MED_TractorFair/Meshes/M_MED_TractorFair_AnimBP
target_skeleton=&/M_MED_TractorFair_Skeleton
graphs [
 AnimGraph {nodes [
   Output Pose(Result=Use cached pose 'dynamics_Cache'()) {layerGroup=None}
   n1 = Control Rig!(Source=Copy Pose From Mesh() {bUseAttachedParent=true}, Alpha, emote:double=Get emote() -> (emote:double)) -> (Pose) {
    alphaBoolBlend {blendOption=Linear}
    controlRigAssetReference {blueprintRigClass=&/M_MED_TractorFair_DeformRig.M_MED_TractorFair_DeformRig_C}
   }
   Save cached pose 'CtrlRig_Cache'(Pose=@n1) {cacheName=CtrlRig_Cache}
   n2 = Get GravityOverride()
   n3 = Anim Dynamics - Bone: dyn_TF_chain_sh_r!(ComponentPose=Local To Component(LocalPose=Use cached pose 'RB_cache'()), Alpha=1, GravityOverride:Vector=@n2, ExternalForce="0,0,0") -> (Pose) {
    angularDampingOverride=1
    angularSpringConstant=20000
    bAngularSpring=true
    bOverrideAngularDamping=true
    bUseGravityOverride=true
    gravityScale=0.35
    simulationSpace=RootRelative
    alphaBoolBlend {blendOption=Linear}
    boundBone {boneName=dyn_TF_chain_sh_r}
    componentLinearAccScale=<1 1 1>
    componentLinearVelScale=<1 1 1>
    gravityOverride {z=-980}
    physicsBodyDefinitions [{
      boundBone {boneName=dyn_TF_chain_sh_r}
      boxExtents=<20 30 30>
      constraintSetup {angularLimitsMax=<2 0 2>, angularLimitsMin=<-2 -60 -2>, angularTarget=<1 0 0>}
      localJointOffset=<4 0 0>
     }]
   }
   n4 = Anim Dynamics - Bone: dyn_TF_chain_sh_front_r!(ComponentPose=@n3, Alpha=1, GravityOverride:Vector=@n2, ExternalForce="0,0,0") -> (Pose) {
    angularDampingOverride=1
    angularSpringConstant=20000
    bAngularSpring=true
    bOverrideAngularDamping=true
    bUseGravityOverride=true
    gravityScale=0.35
    simulationSpace=RootRelative
    alphaBoolBlend {blendOption=Linear}
    boundBone {boneName=dyn_TF_chain_sh_front_r}
    componentLinearAccScale=<1 1 1>
    componentLinearVelScale=<1 1 1>
    gravityOverride {z=-980}
    physicsBodyDefinitions [{
      boundBone {boneName=dyn_TF_chain_sh_front_r}
      boxExtents=<20 30 30>
      constraintSetup {angularLimitsMax=<2 0 2>, angularLimitsMin=<-2 -60 -2>, angularTarget=<1 0 0>}
      localJointOffset=<4 0 0>
     }]
   }
   n5 = Anim Dynamics - Bone: dyn_TF_belt_sh_down_r(ComponentPose=@n4, Alpha=1, ExternalForce="0,0,0") {
    angularDampingOverride=1
    angularSpringConstant=20000
    bAngularSpring=true
    bOverrideAngularDamping=true
    bUseGravityOverride=true
    gravityScale=0.35
    simulationSpace=RootRelative
    alphaBoolBlend {blendOption=Linear}
    boundBone {boneName=dyn_TF_belt_sh_down_r}
    componentLinearAccScale=<1 1 1>
    componentLinearVelScale=<1 1 1>
    gravityOverride {z=-980}
    physicsBodyDefinitions [{
      boundBone {boneName=dyn_TF_belt_sh_down_r}
      boxExtents=<55 55 55>
      constraintSetup {angularLimitsMax=<5 5 5>, angularLimitsMin=<-5 -5 -5>, angularTarget=<1 0 0>}
      localJointOffset=<4 0 0>
     }]
   }
   n6 = Anim Dynamics - Bone: dyn_armband_low!(ComponentPose=@n5, Alpha=Get NOTskydiving() -> (NOTskydiving:double), GravityOverride:Vector="0,0,-980", ExternalForce="0,0,0") -> (Pose) {
    alpha=0.9
    angularDampingOverride=0.98
    angularSpringConstant=1500
    bAngularSpring=true
    bGravityOverrideInSimSpace=true
    bOverrideAngularDamping=true
    bUseGravityOverride=true
    numSolverIterationsPostUpdate=4
    numSolverIterationsPreUpdate=16
    simulationSpace=BoneRelative
    alphaBoolBlend {blendOption=Linear}
    boundBone {boneName=dyn_armband_low}
    componentAppliedLinearAccClamp=<60000 60000 60000>
    componentLinearAccScale=<0.3 0.2 0.2>
    componentLinearVelScale=<0.3 0.2 0.2>
    gravityOverride {z=-980}
    physicsBodyDefinitions [{
      sphereCollisionRadius=0
      boundBone {boneName=dyn_armband_low}
      boxExtents=<15 5 5>
      constraintSetup {angularTargetAxis=AxisZ, twistAxis=AxisY, angularLimitsMax=<25 25 25>, angularLimitsMin=<-25 -25 -25>, angularTarget=<0 0 1>}
      localJointOffset=<2 0 0>
     }]
    relativeSpaceBone {boneName=root}
   }
   n7 = Anim Dynamics - Bone: dyn_armband_hi!(ComponentPose=@n6, Alpha=1, GravityOverride:Vector="0,0,-980", ExternalForce="0,0,0") -> (Pose) {
    alpha=0.9
    angularDampingOverride=0.98
    angularSpringConstant=1500
    bAngularSpring=true
    bGravityOverrideInSimSpace=true
    bOverrideAngularDamping=true
    bUseGravityOverride=true
    numSolverIterationsPostUpdate=4
    numSolverIterationsPreUpdate=16
    simulationSpace=BoneRelative
    alphaBoolBlend {blendOption=Linear}
    boundBone {boneName=dyn_armband_hi}
    componentAppliedLinearAccClamp=<60000 60000 60000>
    componentLinearAccScale=<0.3 0.2 0.2>
    componentLinearVelScale=<0.3 0.2 0.2>
    gravityOverride {z=-980}
    physicsBodyDefinitions [{
      sphereCollisionRadius=0
      boundBone {boneName=dyn_armband_hi}
      boxExtents=<15 5 5>
      constraintSetup {angularTargetAxis=AxisZ, twistAxis=AxisY, angularLimitsMax=<25 25 25>, angularLimitsMin=<-25 -25 -25>, angularTarget=<0 0 1>}
      localJointOffset=<2 0 0>
     }]
    relativeSpaceBone {boneName=root}
   }
   n8 = RigidBody!(
    ComponentPose=Local To Component(LocalPose=Use cached pose 'constraints_cache'())
    Alpha
    OverrideWorldGravity:Vector=Get GravityOverride()
    ExternalForce
   ) -> (Pose) {
    bClampLinearTranslationLimitToRefPose=true
    bOverrideWorldGravity=true
    bUseExternalClothCollision=true
    simulationSpace=BaseBoneSpace
    alphaBoolBlend {blendOption=Linear}
    baseBoneRef {boneName=spine_05}
    componentAppliedLinearAccClamp=<3000 3000 3000>
    componentLinearAccScale=<0.3 0.3 0.3>
    componentLinearVelScale=<0.3 0.3 0.3>
    simSpaceSettings {maxAngularAcceleration=1000, maxAngularVelocity=3, maxLinearAcceleration=700, maxLinearVelocity=500, velocityScaleZ=0.2, worldAlpha=0.25}
   }
   Save cached pose 'dynamics_Cache'(Pose=Blend Poses by bool(BlendPose_0=Component To Local(ComponentPose=@n8), BlendPose_1=Component To Local(ComponentPose=@n7), bActiveValue=Get bIsInFrontEnd()) {blendType=Linear}) {cacheName=dynamics_Cache}
   n9 = "Constraint - base_TF_chain_sh_r (PPPPPPPP)"!(
    ComponentPose=Local To Component(LocalPose=Use cached pose 'CtrlRig_Cache'())
    Alpha=1
    ConstraintWeights_0=0.3829
    ConstraintWeights_1:float=0.2511
    ConstraintWeights_2:float=0.1372
    ConstraintWeights_3:float=0.0927
    ConstraintWeights_4:float=0.0658
    ConstraintWeights_5:float=0.0504
    ConstraintWeights_6:float=0.0197
    ConstraintWeights_7:float=0.0001
   ) -> (Pose) {
    alphaBoolBlend {blendOption=Linear}
    boneToModify {boneName=base_TF_chain_sh_r}
    constraintSetup [
     {targetBone {boneName=clavicle_r}}
     {targetBone {boneName=spine_05}}
     {targetBone {boneName=deform_upperarm_up_r}}
     {targetBone {boneName=upperarm_twist_01_r}}
     {targetBone {boneName=spine_03}}
     {targetBone {boneName=deform_upperarm_bk_r}}
     {targetBone {boneName=spine_04}}
     {targetBone {boneName=spine_02}}
    ]
    constraintWeights [1,1,1,1,1,1,1,1]
   }
   n10 = "Constraint - base_TF_chain_sh_front_r (PPPPPPPP)"!(
    ComponentPose=@n9
    Alpha=1
    ConstraintWeights_0=0.324
    ConstraintWeights_1:float=0.2819
    ConstraintWeights_2:float=0.1077
    ConstraintWeights_3:float=0.0843
    ConstraintWeights_4:float=0.0806
    ConstraintWeights_5:float=0.076
    ConstraintWeights_6:float=0.0417
    ConstraintWeights_7:float=0.0038
   ) -> (Pose) {
    alphaBoolBlend {blendOption=Linear}
    boneToModify {boneName=base_TF_chain_sh_front_r}
    constraintSetup [
     {targetBone {boneName=clavicle_r}}
     {targetBone {boneName=spine_05}}
     {targetBone {boneName=spine_04}}
     {targetBone {boneName=deform_upperarm_up_r}}
     {targetBone {boneName=upperarm_twist_01_r}}
     {targetBone {boneName=deform_upperarm_fr_r}}
     {targetBone {boneName=spine_03}}
     {targetBone {boneName=spine_02}}
    ]
    constraintWeights [1,1,1,1,1,1,1,1]
   }
   n11 = "Constraint - base_armband_knot (PPPPPPPP)"!(
    ComponentPose=@n10
    Alpha=1
    ConstraintWeights_0=0.3976
    ConstraintWeights_1:float=0.2337
    ConstraintWeights_2:float=0.1776
    ConstraintWeights_3:float=0.1
    ConstraintWeights_4:float=0.0498
    ConstraintWeights_5:float=0.0207
    ConstraintWeights_6:float=0.0197
    ConstraintWeights_7:float=0.001
   ) -> (Pose) {
    alphaBoolBlend {blendOption=Linear}
    boneToModify {boneName=base_armband_knot}
    constraintSetup [
     {targetBone {boneName=upperarm_twist_01_l}}
     {targetBone {boneName=dyn_armband_knot}}
     {targetBone {boneName=upperarm_twist_02_l}}
     {targetBone {boneName=upperarm_l}}
     {targetBone {boneName=dyn_armband_hi}}
     {targetBone {boneName=dyn_armband_low}}
     {targetBone {boneName=deform_upperarm_up_l}}
     {targetBone {boneName=clavicle_l}}
    ]
    constraintWeights [1,1,1,1,1,1,1,1]
   }
   Save cached pose 'constraints_cache'(Pose=Component To Local(ComponentPose=@n11)) {cacheName=constraints_cache}
   n12 = RigidBody(ComponentPose=Local To Component(LocalPose=Use cached pose 'constraints_cache'()), OverrideWorldGravity=Get GravityOverride()) {
    bClampLinearTranslationLimitToRefPose=true
    bOverrideWorldGravity=true
    bUseExternalClothCollision=true
    overridePhysicsAsset=&/M_MED_TractorFair_Physics
    simulationSpace=BaseBoneSpace
    alphaBoolBlend {blendOption=Linear}
    baseBoneRef {boneName=pelvis}
    componentAppliedLinearAccClamp=<3000 3000 3000>
    componentLinearAccScale=<0.3 0.3 0.3>
    componentLinearVelScale=<0.3 0.3 0.3>
    simSpaceSettings {maxAngularAcceleration=1000, maxAngularVelocity=3, maxLinearAcceleration=700, maxLinearVelocity=500, velocityScaleZ=0.2, worldAlpha=0.25}
   }
   Save cached pose 'RB_cache'(Pose=Component To Local(ComponentPose=@n12)) {cacheName=RB_cache}
  ]}
 EventGraph {nodes [
   n1 = Event BlueprintUpdateAnimation()
   > n2 = Gravity Override State Function DEPRECATED(CharacterPartAnimInstance=Self-Reference(), GameTimeInSeconds=GetGameTimeInSeconds(), GravityOverrideParams=Get Gravity Override Params())
   > Set GravityOverride(GravityOverride=@n2)
   n3 = Get bIsPlayingEmote()
   > Set emote(emote:double=FInterpTo(Current=Get emote(), Target="To Float (Boolean)"(InBool=@n3), DeltaTime=@n1.DeltaTimeX, InterpSpeed=0.8)) -> (Output_Get:double)
   n4 = NOT Boolean(A=Get bIsSkyDiving())
   > Set NOTskydiving(NOTskydiving:double=FInterpTo(Current=Get NOTskydiving(), Target="To Float (Boolean)"(InBool=@n4), DeltaTime=@n1.DeltaTimeX, InterpSpeed=2)) -> (Output_Get:double)
   n5 = Event OnWeaponChanged()
   n6 = Get Dynamic Material Instance(execute=[@n5.then, @n9.then], SkeletalMeshComponent=GetPartSkeletalMeshComponent())
   n7 = Make Array("[0]"=(TagName="Cosmetics.MetaTags.TractorFair"))
   > n8 = Branch(Condition=OR Boolean(A=PickaxeItemDefinitionHasAnyMetaTag(Tags=@n7, bRequiresEquipped=true), B=Get bIsSkyDiving()))
   > SetScalarParameterValue(self=@n6, ParameterName=Hide Element 02, Value=0)
   > SetScalarParameterValue(self=@n6, ParameterName=Hide Element 01, Value=1)
   SetScalarParameterValue(execute=@n8.else, self=@n6, ParameterName=Hide Element 02, Value=1)
   > SetScalarParameterValue(self=@n6, ParameterName=Hide Element 01, Value=0)
   n9 = Event OnCharacterCustomizationLoaded()
  ]}
]
variables [
 emote {type=number}
 Gravity Override Params {type=GravityOverrideParamsStruct}
 GravityOverride {type=Vector}
 NOTskydiving {type=number}
]
