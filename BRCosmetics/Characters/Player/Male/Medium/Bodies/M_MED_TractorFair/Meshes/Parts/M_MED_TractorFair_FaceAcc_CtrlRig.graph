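# Control Rig for the TractorFair medium male face accessory that corrects back-hair bones as the head rotates.
# A spherical pose reader on the head bone outputs a weight that drives chained ModifyTransforms offsets on the back hair.
# The capeOrBackpack bool switches corrective sets: true offsets the base and a right strand, false the base and two left strands.
# An unconnected "controls" function copies head_ctrl to the head bone.
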ControlRigBlueprint /BRCosmetics/Characters/Player/Male/Medium/Bodies/M_MED_TractorFair/Meshes/Parts/M_MED_TractorFair_FaceAcc_CtrlRig
graphs [
 New Function_ContainedGraph {nodes [
   RigVMFunctionEntryNode()
   > &RigUnit_SetTransform(
    Item=(Type=Bone,Name="head")
    Weight=1
    bInitial=False
    Value=&RigUnit_GetTransform(Item=(Type=Control,Name="head_ctrl"), bInitial=False).Transform
   )
   > RigVMFunctionReturnNode()
  ]}
 RigVMFunctionLibrary {nodes [controls = RigVMCollapseNode(ExecuteContext=())]}
 RigVMModel {nodes [
   RigUnit_BeginExecution = &RigUnit_BeginExecution()
   RigVMFunction_Sequence = &RigVMFunction_Sequence(ExecuteContext=@RigUnit_BeginExecution.ExecutePin)
   n1 = &RigUnit_SphericalPoseReader(
    ActiveRegionScaleFactors=(PositiveWidth=1,NegativeWidth=0.2,PositiveHeight=1.5,NegativeHeight=1.5)
    ActiveRegionSize=0.1
    Debug=(bDrawDebug=false,bDraw2D=false,bDrawLocalAxes=false,DebugScale=25,DebugSegments=20,DebugThickness=0.25)
    DriverAxis=(X=0,Y=1,Z=0)
    DriverItem=(Type=Bone,Name="head")
    FalloffRegionScaleFactors=(PositiveWidth=3,NegativeWidth=1.8,PositiveHeight=2.5,NegativeHeight=2.5)
    FalloffSize=0.2
    FlipHeightScaling=False
    FlipWidthScaling=False
    OptionalParentItem=(Type=Bone,Name="None")
    RotationOffset=(X=90,Y=0,Z=-55)
    ExecutePin=@RigVMFunction_Sequence.B
   )
   RigVMFunction_ControlFlowBranch = &RigVMFunction_ControlFlowBranch(ExecuteContext=@n1.ExecutePin, Condition=RigVMVariableNode() {variable=capeOrBackpack}.Value)
   &RigUnit_ModifyTransforms(
    ItemToModify=((Item=(Type=Bone,Name="base_TF_hair_back_c"),Transform=(Rotation=(X=-0.2802,Y=-0.539,Z=-0.1962,W=0.7698),Translation=(X=-3,Y=0,Z=0),Scale3D=(X=1,Y=1,Z=1))))
    WeightMinimum=0
    ExecutePin=@RigVMFunction_ControlFlowBranch.False
    Weight=@n1.OutputParam
   )
   > &RigUnit_ModifyTransforms(
    ItemToModify=((Item=(Type=Bone,Name="dyn_TF_hair_back_l_04"),Transform=(Rotation=(X=0,Y=0,Z=0,W=1),Translation=(X=0,Y=-15,Z=10),Scale3D=(X=1,Y=1,Z=1))))
    WeightMinimum=0
    Weight=@n1.OutputParam
   )
   > &RigUnit_ModifyTransforms(
    ItemToModify=((Item=(Type=Bone,Name="dyn_TF_hair_back_l_03_a"),Transform=(Rotation=(X=0,Y=0,Z=0,W=1),Translation=(X=0,Y=-15,Z=12.5),Scale3D=(X=1,Y=1,Z=1))))
    WeightMinimum=0
    Weight=@n1.OutputParam
   )
   &RigUnit_ModifyTransforms(
    ItemToModify=((Item=(Type=Bone,Name="base_TF_hair_back_c"),Transform=(Rotation=(X=-0.2802,Y=-0.539,Z=-0.1962,W=0.7698),Translation=(X=-5,Y=10,Z=0),Scale3D=(X=1,Y=1,Z=1))))
    WeightMinimum=0
    ExecutePin=@RigVMFunction_ControlFlowBranch.True
    Weight=@n1.OutputParam
   )
   > &RigUnit_ModifyTransforms(
    ItemToModify=((Item=(Type=Bone,Name="dyn_TF_hair_back_r_03"),Transform=(Rotation=(X=0,Y=0,Z=0,W=1),Translation=(X=0,Y=15,Z=15),Scale3D=(X=1,Y=1,Z=1))))
    WeightMinimum=0
    Weight=@n1.OutputParam
   )
   controls = RigVMFunctionReferenceNode(ExecuteContext=())
  ]}
]
hierarchy [
 root {loc=<0 0 0>, rot=<0 0 0 1>}
 pelvis {parent=root, loc=<0 -0.7071 93.1191>, rot=<0.0127 0.707 -0.0127 -0.707>}
 spine_01 {parent=pelvis, loc=<6.7494 -0.6388 0>, rot=<0 0 0.034 -0.9994>}
 spine_02 {parent=spine_01, loc=<7.2518 -0.035 0>, rot=<0 0 0.0238 -0.9997>}
 spine_03 {parent=spine_02, loc=<6.0104 0 0>, rot=<0 0 0.0149 -0.9999>}
 spine_04 {parent=spine_03, loc=<8.1024 0 0>, rot=<0 0 -0.0132 -0.9999>}
 spine_05 {parent=spine_04, loc=<10.0146 0 0>, rot=<0 0 -0.0015 -1>}
 clavicle_l {parent=spine_05, loc=<4.3575 -0.1466 -2.3223>, rot=<0.0465 -0.7289 -0.009 -0.6829>}
 upperarm_l {parent=clavicle_l, loc=<15.5111 0 0>, rot=<0.0226 -0.3234 -0.0088 -0.9459>}
 neck_01 {parent=spine_05, loc=<11.3138 0.1529 0>, rot=<0 0 -0.2003 -0.9797>}
 neck_02 {parent=neck_01, loc=<4.5627 0 0>, rot=<0 0 0.0504 -0.9987>}
 head {parent=neck_02, loc=<5.3914 0 0>, rot=<0 0 0.115 -0.9934>}
 base_TF_hair_front_r {parent=head, loc=<10.1869 11.0779 1.6>, rot=<0.7048 -0.1104 0.1126 -0.6916>}
 dyn_TF_hair_front_r {parent=base_TF_hair_front_r, loc=<0 0 0>, rot=<0 0 0 1>}
 base_TF_hair_front_l {parent=head, loc=<10.1869 11.0779 -1.6>, rot=<0.7048 -0.1104 0.1126 -0.6916>}
 dyn_TF_hair_front_l {parent=base_TF_hair_front_l, loc=<0 0 0>, rot=<0 0 0 1>}
 base_TF_hair_front_c {parent=head, loc=<7.1915 12.4871 0>, rot=<0.7048 -0.1104 0.1126 -0.6916>}
 dyn_TF_hair_front_c {parent=base_TF_hair_front_c, loc=<0 0 0>, rot=<0 0 0 1>}
 base_TF_hair_back_c {parent=head, loc=<12.3788 -5.2963 0>, rot=<0.1581 -0.683 0.696 -0.1552>}
 dyn_TF_hair_back_c_01 {parent=base_TF_hair_back_c, loc=<0 0 0>, rot=<0 0 0 1>}
 dyn_TF_hair_back_c_02 {parent=dyn_TF_hair_back_c_01, loc=<10.091 0.0074 -0.141>, rot=<0 0.0657 0 -0.9978>}
 dyn_TF_hair_back_c_03 {parent=dyn_TF_hair_back_c_02, loc=<8.8778 0 0>, rot=<0 -0.0867 0 -0.9962>}
 dyn_TF_hair_back_c_04 {parent=dyn_TF_hair_back_c_03, loc=<8.8775 0 0>, rot=<0 -0.1527 0 -0.9883>}
 dyn_TF_hair_back_c_05 {parent=dyn_TF_hair_back_c_04, loc=<8.8776 0 0>, rot=<0 -0.0493 0 -0.9988>}
 dyn_TF_hair_back_c_06 {parent=dyn_TF_hair_back_c_05, loc=<4.7623 0 0>, rot=<0 0 0 1>}
 dyn_TF_hair_back_l_04 {parent=dyn_TF_hair_back_c_04, loc=<-0.2576 4.9861 -1.1768>, rot=<0 0 0 1>}
 dyn_TF_hair_back_r_03 {parent=dyn_TF_hair_back_c_03, loc=<0.3908 -4.4461 -2.446>, rot=<0 0 0 1>}
 dyn_TF_hair_back_l_03_a {parent=dyn_TF_hair_back_c_03, loc=<-0.7734 4.3638 -3.3802>, rot=<0 0 0 1>}
 dyn_TF_hair_back_l_03_b {parent=dyn_TF_hair_back_c_03, loc=<-2.2121 3.5287 -0.2834>, rot=<0 0 0 1>}
 dyn_TF_hair_back_l_02 {parent=dyn_TF_hair_back_c_02, loc=<-0.0427 3.5389 -1.443>, rot=<0 0 0 1>}
 dyn_TF_hair_back_r_02 {parent=dyn_TF_hair_back_c_02, loc=<0.9367 -5.528 -2.9118>, rot=<0 0 0 1>}
 clavicle_r {parent=spine_05, loc=<4.3575 -0.1466 2.3223>, rot=<0.7289 0.0465 -0.6829 0.009>}
 upperarm_r {parent=clavicle_r, loc=<-15.5111 0 0>, rot=<0.0226 -0.3234 -0.0088 -0.9459>}
 head_ctrl {controlType=EULER_TRANSFORM, type=Control, loc=<0 0 0>, rot=<0 0 0 1>}
]
variables [capeOrBackpack {type=bool, default=False}]
