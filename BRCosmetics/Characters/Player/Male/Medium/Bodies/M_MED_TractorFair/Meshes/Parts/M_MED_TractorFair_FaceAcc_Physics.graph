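# Physics asset for the TractorFair male medium cosmetic's face accessory part, setting up physics for a back hair chain.
# Capsule bodies cover the upper spine, clavicles, upper arms, neck and head, giving a body volume around the hair.
# Six dyn hair bones chain from a small base body on the head, with limited swing and very tight twist.
# Constraints between body bones, and from hair base to head, leave all angular motion free.
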PhysicsAsset /BRCosmetics/Characters/Player/Male/Medium/Bodies/M_MED_TractorFair/Meshes/Parts/M_MED_TractorFair_FaceAcc_Physics
bodies {
 base_TF_hair_back_c [{shapeType=Capsule, length=0.505, radius=0.505, shapeName=base_TF_hair_back_c_capsule, rotation=<rot 0 -90 0>}]
 clavicle_l [{shapeType=Capsule, length=10.342, radius=10.342, shapeName=clavicle_l_capsule, center=<8.937 0.7626 2.0448>, rotation=<rot 90 180 90>}]
 clavicle_r [{shapeType=Capsule, length=10.342, radius=9.5, shapeName=clavicle_r_capsule, center=<-6.2656 0.9249 -2.5648>, rotation=<rot 90 0 0>}]
 dyn_TF_hair_back_c_01 [{shapeType=Capsule, length=10, radius=3, shapeName=dyn_TF_hair_back_c_01_capsule, center=<3 0 1.2106>, rotation=<rot -90 -165.9638 -104.0362>}]
 dyn_TF_hair_back_c_02 [{shapeType=Capsule, length=4, radius=4, shapeName=dyn_TF_hair_back_c_02_capsule, center=<3 0 0.3897>, rotation=<rot -90 180 -90>}]
 dyn_TF_hair_back_c_03 [{shapeType=Capsule, length=6, radius=3, shapeName=dyn_TF_hair_back_c_03_capsule, center=<3 0 0.6641>, rotation=<rot -90 -165.9638 -104.0362>}]
 dyn_TF_hair_back_c_04 [{shapeType=Capsule, length=6, radius=3, shapeName=dyn_TF_hair_back_c_04_capsule, center=<3 0 0.8327>, rotation=<rot -90 180 -90>}]
 dyn_TF_hair_back_c_05 [{shapeType=Capsule, length=4, radius=3, shapeName=dyn_TF_hair_back_c_05_capsule, center=<3 0 0.7048>, rotation=<rot -90 180 180>}]
 dyn_TF_hair_back_c_06 [{shapeType=Capsule, length=4, radius=3, shapeName=dyn_TF_hair_back_c_06_capsule, center=<3 0 1.4315>, rotation=<rot -90 0 0>}]
 head [{shapeType=Capsule, length=9.1393, radius=9.1393, shapeName=head_capsule, center=<0.4651 2.3372 0>, rotation=<rot -90 0 0>}]
 neck_01 [{shapeType=Capsule, length=7.5316, radius=7.5316, shapeName=neck_01_capsule, center=<0.8569 -1.0916 0>}]
 neck_02 [{shapeType=Capsule, length=6.9487, radius=6.9487, shapeName=neck_02_capsule, center=<0 -0.5978 0>}]
 spine_03 [{shapeType=Capsule, length=5.7806, radius=15.297, shapeName=spine_04_capsule, center=<0 3.2677 0>}]
 spine_04 [{shapeType=Capsule, length=15.297, radius=15.297, shapeName=spine_04_capsule, center=<5.7463 0.6145 0>}]
 spine_05 [{shapeType=Capsule, length=16.03, radius=16.03, shapeName=spine_05_capsule, center=<2.6035 2.4222 0>}]
 upperarm_l [{shapeType=Capsule, length=29.2002, radius=7.2998, shapeName=upperarm_l_capsule, center=<15.0285 -0.5625 0>, rotation=<rot 90 0 0>}]
 upperarm_r [{
   shapeType=Capsule
   length=30.2002
   radius=7.7998
   shapeName=upperarm_r_capsule
   center=<-15.0285 1.3524 0.3067>
   rotation=<rot 90 -0.0031 -0.0039>
  }]
}
constraints [
 {
  bone1Name=clavicle_l
  bone2Name=spine_05
  swing1LimitDegrees=45
  swing1Motion=Free
  swing2LimitDegrees=45
  swing2Motion=Free
  twistLimitDegrees=45
  twistMotion=Free
 }
 {
  bone1Name=neck_01
  bone2Name=spine_05
  swing1LimitDegrees=45
  swing1Motion=Free
  swing2LimitDegrees=45
  swing2Motion=Free
  twistLimitDegrees=45
  twistMotion=Free
 }
 {
  bone1Name=neck_02
  bone2Name=neck_01
  swing1LimitDegrees=45
  swing1Motion=Free
  swing2LimitDegrees=45
  swing2Motion=Free
  twistLimitDegrees=45
  twistMotion=Free
 }
 {
  bone1Name=head
  bone2Name=neck_02
  swing1LimitDegrees=45
  swing1Motion=Free
  swing2LimitDegrees=45
  swing2Motion=Free
  twistLimitDegrees=45
  twistMotion=Free
 }
 {
  bone1Name=clavicle_r
  bone2Name=spine_05
  swing1LimitDegrees=45
  swing1Motion=Free
  swing2LimitDegrees=45
  swing2Motion=Free
  twistLimitDegrees=45
  twistMotion=Free
 }
 {
  bone1Name=base_TF_hair_back_c
  bone2Name=head
  swing1LimitDegrees=45
  swing1Motion=Free
  swing2LimitDegrees=45
  swing2Motion=Free
  twistLimitDegrees=45
  twistMotion=Free
 }
 {bone1Name=dyn_TF_hair_back_c_01, bone2Name=base_TF_hair_back_c, swing1LimitDegrees=20, swing2LimitDegrees=25, twistLimitDegrees=1}
 {bone1Name=dyn_TF_hair_back_c_02, bone2Name=dyn_TF_hair_back_c_01, swing1LimitDegrees=20, swing2LimitDegrees=45, twistLimitDegrees=1}
 {bone1Name=dyn_TF_hair_back_c_03, bone2Name=dyn_TF_hair_back_c_02, swing1LimitDegrees=20, swing2LimitDegrees=45, twistLimitDegrees=1}
 {bone1Name=dyn_TF_hair_back_c_04, bone2Name=dyn_TF_hair_back_c_03, swing1LimitDegrees=20, swing2LimitDegrees=45, twistLimitDegrees=1}
 {bone1Name=dyn_TF_hair_back_c_06, bone2Name=dyn_TF_hair_back_c_05, swing1LimitDegrees=20, swing2LimitDegrees=45, twistLimitDegrees=1}
 {bone1Name=dyn_TF_hair_back_c_05, bone2Name=dyn_TF_hair_back_c_04, swing1LimitDegrees=20, swing2LimitDegrees=45, twistLimitDegrees=1}
 {
  bone1Name=spine_05
  bone2Name=spine_04
  swing1LimitDegrees=45
  swing1Motion=Free
  swing2LimitDegrees=45
  swing2Motion=Free
  twistLimitDegrees=45
  twistMotion=Free
 }
 {
  bone1Name=upperarm_l
  bone2Name=clavicle_l
  swing1LimitDegrees=45
  swing1Motion=Free
  swing2LimitDegrees=45
  swing2Motion=Free
  twistLimitDegrees=45
  twistMotion=Free
 }
 {
  bone1Name=upperarm_r
  bone2Name=clavicle_r
  swing1LimitDegrees=45
  swing1Motion=Free
  swing2LimitDegrees=45
  swing2Motion=Free
  twistLimitDegrees=45
  twistMotion=Free
 }
 {
  bone1Name=spine_04
  bone2Name=spine_03
  swing1LimitDegrees=45
  swing1Motion=Free
  swing2LimitDegrees=45
  swing2Motion=Free
  twistLimitDegrees=45
  twistMotion=Free
 }
]
